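; Conversation task that lets players fund a War Effort funding station through its conversation service.
; On execute, it casts the NPC participant to the station, adds the row-driven price to the chosen fund index, then advances or aborts the conversation.
; Static extra data packs choice count, funds, limit, icons, names and station index into string key/value pairs.
; Displayed funds are false-clamped below the limit when the initiative prevents mid-match completion.
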(fn ExecuteTaskNode
  (params (Context :type Conversation%20Context%20Structure%20%28by%20ref%29))
  (bind _fundchoiceindex (Variables|Default|GetFundChoiceIndex))
  (bind _returnvalue (Utilities|String|ToString(Integer) _fundchoiceindex))
  (bind _row_fundingprice (Variables|Default|GetRow_FundingPrice))
  (bind _returnvalue_2 (ScalableFloat|GetValueatLevel _row_fundingprice))
  (bind _returnvalue_3 (Utilities|String|Append "Fund Choice Index:  " _returnvalue))
  (bind _fundchoiceindex_2 (Variables|Default|GetFundChoiceIndex))
  (bind _output_get (LocalVariables|SetConversationContext Context))
  (bind _returnvalue_4 (|CallParentFunction _output_get))
  (bind _returnvalue_5 (Conversation|CanConversationContinue _returnvalue_4))
  (if _returnvalue_5
    (Utilities|FlowControl|Sequence
      (:then_0
        ;[1 1 1] Get Funding Component/Actor to call On Funds Added +
        ; Get Funding Station Num
        ; Could this function + data all live ON the component to keep this more reusable on more actor types?
        (bind _returnvalue_6 (Class|FortConversationContextLibrary|GetParticipantActorNPC (LocalVariables|GetConversationContext)))
        (bind _asb_war_effort_funding_station_base (Utilities|Casting|CastToB_WarEffort_FundingStation_Base _returnvalue_6))
        (Utilities|IsValid _asb_war_effort_funding_station_base
          (:Is%20Valid
            (Development|PrintString _returnvalue_3)
            (Fortnite|FortShippingLog self _returnvalue_3)
            (Fortnite|FortShippingLog self (Utilities|String|ToString(Text) (Utilities|Text|FormatText (Variables|Default|GetFundingMessageFormat) (Utilities|String|ToString(Object) _asb_war_effort_funding_station_base) _returnvalue_2 _fundchoiceindex_2)))
            (Funding|AddtoStationFund _asb_war_effort_funding_station_base (Math|Float|TruncatetoInteger64 _returnvalue_2) _fundchoiceindex_2)
            (LocalVariables|SetSucceeded true))))
      (:then_1
        ;[1 1 1] Hook in Succeeded here once owner is correct actor type / has correct component type
        (if (LocalVariables|GetSucceeded)
          (bind _returnvalue_7 (Conversation|AdvanceConversation (LocalVariables|GetConversationContext)))
          (return _returnvalue_7)
          (else
            (bind _returnvalue_8 (Conversation|AbortConversation (LocalVariables|GetConversationContext)))
            (return _returnvalue_8)))))
    (else
      (return _returnvalue_4))))

(fn GatherStaticExtraData
  (params (Context :type Conversation%20Context%20Structure%20%28by%20ref%29))
  (bind _ioedata (LocalVariables|GetIOEData))
  (bind _ioedata_2 (LocalVariables|GetIOEData))
  (bind _ioedata_3 (LocalVariables|GetIOEData))
  (bind _inoutextradata (|CallParentFunction Context))
  (LocalVariables|SetIOEData _inoutextradata)
  (bind _returnvalue (Class|FortConversationContextLibrary|GetParticipantActorNPC Context))
  (bind _asb_war_effort_funding_station_base (Utilities|Casting|CastToB_WarEffort_FundingStation_Base _returnvalue))
  (Utilities|FlowControl|Sequence
    (:then_0
      (bind (_conversationnodeparameterpair _ismultichoice) (CallFunction|GetIsSingleorMultiEntry _asb_war_effort_funding_station_base))
      (Utilities|Array|Add (LocalVariables|GetIOEData) _conversationnodeparameterpair)
      (LocalVariables|SetIsMultiChoice _ismultichoice))
    (:then_1
      (bind (_conversationnodeparameterpair_0 _conversationnodeparameterpair_1) (CallFunction|GetCurrentFundingAmount _asb_war_effort_funding_station_base (LocalVariables|GetIsMultiChoice)))
      (Utilities|Array|Add _ioedata_2 _conversationnodeparameterpair_0)
      (Utilities|Array|Add _ioedata_2 _conversationnodeparameterpair_1))
    (:then_2
      (bind _conversationnodeparameterpair_2 (CallFunction|GetMaxRequiredFundingAmount _asb_war_effort_funding_station_base))
      (Utilities|Array|Add (LocalVariables|GetIOEData) _conversationnodeparameterpair_2))
    (:then_3
      (bind (_conversationnodeparameterpair_0_2 _conversationnodeparameterpair_1_2) (CallFunction|GetItemIcons _asb_war_effort_funding_station_base (LocalVariables|GetIsMultiChoice)))
      (Utilities|Array|Add _ioedata _conversationnodeparameterpair_0_2)
      (Utilities|Array|Add _ioedata _conversationnodeparameterpair_1_2))
    (:then_4
      (bind (_conversationnodeparameterpair_0_3 _conversationnodeparameterpair_1_3) (CallFunction|GetItemNames _asb_war_effort_funding_station_base (LocalVariables|GetIsMultiChoice)))
      (Utilities|Array|Add _ioedata_3 _conversationnodeparameterpair_0_3)
      (Utilities|Array|Add _ioedata_3 _conversationnodeparameterpair_1_3))
    (:then_5
      (bind _conversationnodeparameterpair_3 (CallFunction|GetNPCIcon _asb_war_effort_funding_station_base))
      (Utilities|Array|Add (LocalVariables|GetIOEData) _conversationnodeparameterpair_3))
    (:then_6
      (bind _conversationnodeparameterpair_4 (CallFunction|GetStationIndex _asb_war_effort_funding_station_base))
      (Utilities|Array|Add (LocalVariables|GetIOEData) _conversationnodeparameterpair_4))
    (:then_7
      (return (LocalVariables|GetIOEData)))))

(fn GetItemIcons
  (params (FundingStation :type B%20War%20Effort%20Funding%20Station%20Base%20Object%20Reference) (IsMultiChoice :type Boolean))
  (bind _parameterkey (LocalVariables|GetParameterKey))
  (bind _returnvalue (Utilities|String|Append _parameterkey "1"))
  (bind _parameterkey_2 (LocalVariables|GetParameterKey))
  (bind _returnvalue_2 (Utilities|String|Append _parameterkey_2 "0"))
  (bind _icons (Getters|GetItemIcons FundingStation))
  (if (Utilities|IsValidSoftObjectReference (Utilities|Array|Get(acopy) _icons))
    (LocalVariables|SetToReturn_0 (Utilities|Struct|MakeConversationNodeParameterPair _returnvalue_2 (Utilities|ToString(SoftObjectReference) (Utilities|Array|Get(acopy) _icons))))
    (else
      (LocalVariables|SetToReturn_0 (Utilities|Struct|MakeConversationNodeParameterPair _returnvalue_2))))
  (if IsMultiChoice
    (if (Utilities|IsValidSoftObjectReference (Utilities|Array|Get(acopy) _icons 1))
      (LocalVariables|SetToReturn_1 (Utilities|Struct|MakeConversationNodeParameterPair _returnvalue (Utilities|ToString(SoftObjectReference) (Utilities|Array|Get(acopy) _icons 1))))
      (else
        (LocalVariables|SetToReturn_1 (Utilities|Struct|MakeConversationNodeParameterPair _returnvalue))))
    (return (LocalVariables|GetToReturn_0) (LocalVariables|GetToReturn_1))))

(fn GetItemNames
  (params (FundingStation :type B%20War%20Effort%20Funding%20Station%20Base%20Object%20Reference) (IsMultiChoice :type Boolean))
  (bind _itemnames (Getters|GetItemNames FundingStation))
  (if (Utilities|Array|IsValidIndex _itemnames)
    (LocalVariables|SetPair0 (Utilities|Struct|MakeConversationNodeParameterPair (Utilities|String|Append (LocalVariables|GetParameterKey) "0") (WarEffort|WriteTexttoBuffer (Utilities|Array|Get(acopy) _itemnames))))
    (else
      (LocalVariables|SetPair0 (Utilities|Struct|MakeConversationNodeParameterPair (Utilities|String|Append (LocalVariables|GetParameterKey) "0")))))
  (if (Utilities|Array|IsValidIndex _itemnames 1)
    (LocalVariables|SetPair1 (Utilities|Struct|MakeConversationNodeParameterPair (Utilities|String|Append (LocalVariables|GetParameterKey) "1") (WarEffort|WriteTexttoBuffer (Utilities|Array|Get(acopy) _itemnames 1))))
    (else
      (LocalVariables|SetPair1 (Utilities|Struct|MakeConversationNodeParameterPair (Utilities|String|Append (LocalVariables|GetParameterKey) "1")))))
  (return (LocalVariables|GetPair0) (LocalVariables|GetPair1)))

(fn GetCurrentFundingAmount
  (params (FundingStation :type B%20War%20Effort%20Funding%20Station%20Base%20Object%20Reference) (IsMultipleChoice :type Boolean))
  (bind _fundingmanagerref (Variables|/WarEffort/Stations/B_WarEffort_FundingStation_Base.B_WarEffort_FundingStation_Base_C|GetFundingManagerRef FundingStation))
  (bind _wareffortconfigdata (Variables|/WarEffort/MeshNetwork/B_WarEffort_FundingManager.B_WarEffort_FundingManager_C|GetWarEffortConfigData _fundingmanagerref))
  (bind _outfunds (Getters|GetCurrentFunds FundingStation))
  (bind (_initiativefound _initiative) (Class|WarEffortConfigData|GetInitiativebyTag _wareffortconfigdata (Variables|/WarEffort/Stations/B_WarEffort_FundingStation_Base.B_WarEffort_FundingStation_Base_C|GetActiveInitiative FundingStation)))
  (LocalVariables|SetLocalShouldFalseClamp (.PreventMidmatchCompletion_27_1970EEF64ADF87A56EDE3CAF8284DCC7 (Utilities|Struct|BreakFWarEffortInitiativeConfig _initiative)))
  (bind (_fundinglimitfound _fundinglimit) (Class|WarEffortConfigData|GetFundingLimitbyFundingType _wareffortconfigdata (Utilities|Struct|BreakFWarEffortInitiativeConfig _initiative)))
  (LocalVariables|SetLocalFundingLimit (Math|Float|TruncatetoInteger64 _fundinglimit))
  (bind _conversationnodeparameterpair (CallFunction|GetIndividualCurrentFundingAmount (LocalVariables|GetParameterKey) _outfunds 0 (LocalVariables|GetLocalShouldFalseClamp) (LocalVariables|GetLocalFundingLimit)))
  (LocalVariables|SetPair0 _conversationnodeparameterpair)
  (bind _conversationnodeparameterpair_2 (CallFunction|GetIndividualCurrentFundingAmount (LocalVariables|GetParameterKey) _outfunds 1 (LocalVariables|GetLocalShouldFalseClamp) (LocalVariables|GetLocalFundingLimit)))
  (LocalVariables|SetPair1 _conversationnodeparameterpair_2)
  (return (LocalVariables|GetPair0) (LocalVariables|GetPair1)))

(fn GetIndividualCurrentFundingAmount
  (params (ParameterKey :type String) (FundingArray :type Array%20of%20Integer64s) (Index :type Integer) (ShouldFalseClamp :type Boolean) (MaxFunding :type Integer64))
  (if (Utilities|Array|IsValidIndex FundingArray Index)
    (return (Utilities|Struct|MakeConversationNodeParameterPair (Utilities|String|Append (Variables|GetParameterKey) (Utilities|String|ToString(Integer) (Variables|GetIndex))) (Utilities|String|ToString(Text) (Utilities|Text|ToText(Integer64) (select (and (Variables|GetShouldFalseClamp) (>= (Utilities|Array|Get(acopy) FundingArray (Variables|GetIndex)) (Variables|GetMaxFunding))) (Math|Conversions|ToInteger64(Float) (* (Math|Conversions|ToFloat(Integer64) (Variables|GetMaxFunding)) (Variables|Default|GetFalseClampPercent))) (Utilities|Array|Get(acopy) FundingArray (Variables|GetIndex)))))))
    (else
      (return (Utilities|Struct|MakeConversationNodeParameterPair (Utilities|String|Append (Variables|GetParameterKey) (Utilities|String|ToString(Integer) (Variables|GetIndex))))))))

(fn GetIsSingleOrMultiEntry
  (params (FundingStation :type B%20War%20Effort%20Funding%20Station%20Base%20Object%20Reference))
  (bind (_initiativefound _options) (Getters|GetInitiativeOptions FundingStation))
  (if (> (Utilities|Array|Length _options) 1)
    (return (Utilities|Struct|MakeConversationNodeParameterPair (LocalVariables|GetParameterKey) (Utilities|String|ToString(Boolean) true)) true)
    (else
      (return (Utilities|Struct|MakeConversationNodeParameterPair (LocalVariables|GetParameterKey) (Utilities|String|ToString(Boolean) false)) false))))

(fn GetMaxRequiredFundingAmount
  (params (FundingStation :type B%20War%20Effort%20Funding%20Station%20Base%20Object%20Reference))
  (bind _fundinglimit (Getters|GetMaxFunds FundingStation))
  (return (Utilities|Struct|MakeConversationNodeParameterPair (LocalVariables|GetParameterKey) (Utilities|String|ToString(Float) _fundinglimit))))

(fn GetNPCIcon
  (params (FundingStation :type B%20War%20Effort%20Funding%20Station%20Base%20Object%20Reference))
  (return (Utilities|Struct|MakeConversationNodeParameterPair (LocalVariables|GetParameterKey) (Utilities|ToString(SoftObjectReference) (Variables|Default|GetServiceGiverIcon)))))

(fn GetStationIndex
  (params (FundingStation :type B%20War%20Effort%20Funding%20Station%20Base%20Object%20Reference))
  (return (Utilities|Struct|MakeConversationNodeParameterPair (LocalVariables|GetParameterKey) (Utilities|String|ToString(Integer) (Getters|GetFundingStationIndex FundingStation)))))
